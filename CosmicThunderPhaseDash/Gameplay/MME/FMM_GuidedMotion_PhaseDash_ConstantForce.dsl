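; Guided motion movement mode for Phase Dash's constant-force motion, overriding how its maximum duration is calculated.
; When the runtime data carries optional networked floats, it safe-divides the first by the constant-force config's scalable Strength to get the duration, so duration scales inversely with Strength.
; With no optional floats, it falls back to the parent implementation.
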(fn GetMaxDuration
  (params (GuidedMotionRuntimeData :type Fort%20Movement%20Mode%20Guided%20Motion%20Runtime%20Data%20Object%20Reference))
  (bind _netoptionalfloats (Variables|/Script/FortniteGame.FortMovementMode_GuidedMotionRuntimeData|GetNetOptionalFloats GuidedMotionRuntimeData))
  (if (not (Utilities|Array|IsEmpty _netoptionalfloats))
    (return (Math|Float|SafeDivide (Utilities|Struct|BreakFortMovementModeFloatPropertyNetworked (Utilities|Array|Get(acopy) _netoptionalfloats)) (ScalableFloat|GetValueatLevel (Variables|Config-ConstantForce|GetStrength))))
    (else
      (bind _returnvalue (|CallParentFunction GuidedMotionRuntimeData))
      (return _returnvalue))))
